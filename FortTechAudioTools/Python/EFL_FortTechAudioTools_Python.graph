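# Editor utility function library for FortTechAudioTools that exposes Python-backed file helpers to Blueprints.
# Each function embeds a script in an Execute Python Script node, passing arguments in and returning results.
# Tkinter dialogs pick files or directories, os.startfile opens files, and csv.DictReader checks CSV headers.
# The header check catches exceptions to return a message, while other helpers let exits or exceptions determine the success flag.
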EditorUtilityBlueprint /FortTechAudioTools/Python/EFL_FortTechAudioTools_Python
graphs [
 FindHeaderInCSVFile {nodes [
   n1 = FindHeaderInCSVFile!() -> (then, CsvFilePath:String, HeaderName:String) {is=title K2Node_FunctionEntry}
   > n2 = Execute Python Script!(
    execute
    PythonScript="## Copyright Epic Games, Inc. All Rights Reserved.\r\n\r\nimport csv\r\nimport unreal\r\nfrom pathlib import Path\r\n\r\n\r\n\r\nif __name__ == '__main__':\r\n    \r\n    with unreal.ScopedEditorTransaction(\"Read CSV File\"):\r\n    \r\n        ## Type-hint Blueprint I/O.\r\n        csv_file_path: str\r\n        header_name: str\r\n        failure_msg_output: str = \"\"\r\n        success: bool = False\r\n        \r\n        ## Try/Catch is used here in order to safely return the failure reason back to Blueprints for widget display.\r\n        ##   -  Blueprints will cancel its Python argument marshalling if an exception is uncaught, which\r\n        ##      will cause 'failure_msg_output' to be undefined when accessed from Blueprints.\r\n        try:\r\n            \r\n            if not header_name:\r\n                raise NameError(\"No header name provided.\")\r\n            \r\n            if not csv_file_path:\r\n                raise TypeError(f\"Provided .csv file path is invalid: '{csv_file_path}'.\")\r\n\r\n            with Path(csv_file_path).open() as csv_file:\r\n                \r\n                csv_reader = csv.DictReader(csv_file, delimiter=',')\r\n                \r\n                if header_name not in csv_reader.fieldnames:\r\n                    raise ValueError(f\"Header name not detected in .csv file: '{header_name}'\")\r\n                \r\n                success = True\r\n        \r\n        except Exception as e:\r\n            failure_msg_output = str(e)\r\n            "
    csv_file_path:String=@n1
    header_name:String=@n1.HeaderName
   ) -> (then, ReturnValue, success:bool=~, failure_msg_output:String) {inputs [csv_file_path, header_name], outputs [success, failure_msg_output]}
   > n3 = Branch(Condition=@n2)
   > Return Node!(execute, Found:bool=@n2.success, OutputMsg:String=@n2.failure_msg_output)
   Return Node(execute=@n3.else, Found=false, OutputMsg=Python script failure.)
  ]}
 OpenDirectorySelector {nodes [
   n1 = OpenDirectorySelector!() -> (then, WindowTitle:String, InitialSearchPath:String) {is=title K2Node_FunctionEntry}
   > n2 = Execute Python Script!(
    execute
    PythonScript="## Copyright Epic Games, Inc. All Rights Reserved.\r\n\r\nimport os\r\nimport sys\r\nimport unreal\r\nfrom   tkinter import filedialog\r\n\r\nwith unreal.ScopedEditorTransaction(\"Directory Selector\"):\r\n\r\n    ## Forward declaration of values passed in from Blueprints.\r\n    window_title:str\r\n    initial_search_path: str\r\n\r\n    ## Prompt user to select directory.\r\n    selected_directory:str = filedialog.askdirectory(initialdir=initial_search_path, mustexist=True, title=f\"{window_title}\")\r\n    \r\n    ## Directory is empty, user cancelled operation.\r\n    if not selected_directory:\r\n        unreal.log(f\"Operation was cancelled by user.\")\r\n        sys.exit(\"Operation was cancelled by user.\")\r\n    \r\n    ## Directory does not exist, somehow.\r\n    if not os.path.exists(selected_directory):\r\n        unreal.log_error(f\"Invalid directory.\")\r\n        unreal.EditorDialog.show_message(title=unreal.Text(\"Error\"), message=unreal.Text(\"Selected directory is invalid.\"), message_type=unreal.AppMsgType.OK)\r\n        sys.exit(\"Selected directory path is invalid.\")\r\n"
    window_title:String=@n1
    initial_search_path:String=@n1.InitialSearchPath
   ) -> (then, ReturnValue, selected_directory:String) {inputs [window_title, initial_search_path], outputs [selected_directory]}
   > Return Node!(execute, Success:bool=@n2, SelectedDirectory:String=@n2.selected_directory)
  ]}
 OpenFileInDefaultApplication {nodes [
   n1 = OpenFileInDefaultApplication!() -> (then, FilePath:String) {is=title K2Node_FunctionEntry}
   > n2 = Execute Python Script!(
    execute
    PythonScript="## Copyright Epic Games, Inc. All Rights Reserved.\r\n\r\nimport os\r\n\r\ndef open_file_in_default_application(input_file_path: str):\r\n    \r\n    if not input_file_path:\r\n        raise ValueError(\"File path field is empty.\")\r\n    \r\n    if not os.path.exists(input_file_path):\r\n        raise ValueError(f\"File path '{input_file_path}' is invalid.\")\r\n    \r\n    os.startfile(input_file_path)\r\n\r\nif __name__ == '__main__':\r\n    \r\n    open_file_in_default_application(file_path) ##type: ignore"
    file_path:String=@n1
   ) -> (then, ReturnValue) {inputs [file_path]}
   > Return Node!(execute, Success=@n2)
  ]}
 OpenFileSelector {nodes [
   n1 = OpenFileSelector!() -> (then, WindowTitle:String, InitialFileSearchPath:String, FileTypeLabel:String, FileTypeExtension:String) {is=title K2Node_FunctionEntry}
   > n2 = Execute Python Script!(
    execute
    PythonScript="## Copyright Epic Games, Inc. All Rights Reserved.\r\n\r\nimport os\r\nimport sys\r\nimport unreal\r\nfrom   tkinter import filedialog\r\n\r\nwith unreal.ScopedEditorTransaction(\"File Selector\"):\r\n    \r\n    ## Forward declaration of values passed in from Blueprints.\r\n    window_title:str\r\n    initial_file_search_path: str\r\n    file_type_label:str\r\n    file_type_extension:str\r\n    \r\n    ## Prompt user to select file path.\r\n    selected_file_path:str = filedialog.askopenfilename(title=window_title, initialdir=initial_file_search_path, filetypes=[(file_type_label, f\"*{file_type_extension}\")])\r\n    \r\n    ## Selected path is empty, user cancelled operation.\r\n    if not selected_file_path:\r\n        unreal.log(f\"Operation was cancelled by user.\")\r\n        sys.exit(\"Operation was cancelled by user.\")\r\n    \r\n    ## Selected file does not exist or is invalid.\r\n    if not os.path.exists(selected_file_path):\r\n        unreal.log_error(f\"Invalid CSV file path.\")\r\n        unreal.EditorDialog.show_message(title=unreal.Text(\"Error\"), message=unreal.Text(\"Selected file path is invalid.\"), message_type=unreal.AppMsgType.OK)\r\n        sys.exit(\"Selected file path is invalid.\")\r\n"
    window_title:String=@n1
    initial_file_search_path:String=@n1.InitialFileSearchPath
    file_type_label:String=@n1.FileTypeLabel
    file_type_extension:String=@n1.FileTypeExtension
   ) -> (then, ReturnValue, selected_file_path:String) {inputs [window_title, initial_file_search_path, file_type_label, file_type_extension], outputs [selected_file_path]}
   > Return Node!(execute, Success=@n2, SelectedFilePath:String=@n2.selected_file_path)
  ]}
]
variables []
